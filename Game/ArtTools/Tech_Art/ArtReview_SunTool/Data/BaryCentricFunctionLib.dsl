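; Barycentric math function library in the ArtReview SunTool data, converting points into and out of triangle and four-point square coordinates.
; ConvertIntoBarycentric compares two cross-product normals, printing a not-coplanar message on failure, then derives UVW from sub-triangle area ratios and tests their sum for validity.
; ConvertOutOfBarycentric rebuilds points as weighted vertex sums.
; The square variants split ABCD into triangles ABC and BCD, using a boolean to track which applies.
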(fn CalculateTriangleArea
  (params (TrianglePointA :type Vector) (TrianglePointB :type Vector) (TrianglePointC :type Vector))
  (return (/ (Math|Vector|VectorLength (Math|Vector|CrossProduct (- TrianglePointB TrianglePointA) (- TrianglePointC TrianglePointA))) 2.0)))

(fn ConvertSquareBaryCentricIN
  (params (TargetPoint :type Vector) (SourceA :type Vector) (SourceB :type Vector) (SourceC :type Vector) (SourceD :type Vector))
  (bind (_uvwcoords _arevalidcoordinates) (CallFunction|ConvertIntoBarycentric TargetPoint SourceA SourceB SourceC))
  (if _arevalidcoordinates
    (return _uvwcoords true true)
    (else
      (bind (_uvwcoords_2 _arevalidcoordinates_2) (CallFunction|ConvertIntoBarycentric TargetPoint SourceB SourceC SourceD))
      (if _arevalidcoordinates_2
        (return _uvwcoords_2 false true)
        (else
          (return _uvwcoords_2 false false))))))

(fn ConvertSquareBaryCentricOUT
  (params (TargetPoint :type Vector) (SourceA :type Vector) (SourceB :type Vector) (SourceC :type Vector) (SourceD :type Vector) (IsTriangleOne? :type Boolean))
  (if IsTriangleOne?
    (return (CallFunction|ConvertOutOfBarycentric TargetPoint SourceA SourceB SourceC))
    (else
      (return (CallFunction|ConvertOutOfBarycentric TargetPoint SourceB SourceC SourceD)))))

(fn ConvertIntoBarycentric
  (params (PointPosition :type Vector) (TrianglePointA :type Vector) (TrianglePointB :type Vector) (TrianglePointC :type Vector))
  (if (Math|Vector|Equal(Vector) (Math|Vector|Normalize (Math|Vector|CrossProduct (- TrianglePointC PointPosition) (- TrianglePointA PointPosition))) (Math|Vector|Normalize (Math|Vector|CrossProduct (- TrianglePointA PointPosition) (- TrianglePointB PointPosition))) 0.002)
    ;[1 1 1] Calculate Areas
    (Utilities|FlowControl|Sequence
      (:then_0
        ;;[1 1 1] Area of ABC
        (LocalVariables|SetAreaOfABC (CallFunction|CalculateTriangleArea TrianglePointA TrianglePointB TrianglePointC)))
      (:then_1
        ;;[1 1 1] Area of CAP
        (LocalVariables|SetAreaOfCAP (CallFunction|CalculateTriangleArea TrianglePointC TrianglePointA PointPosition)))
      (:then_2
        ;;[1 1 1] Area of ABP
        (LocalVariables|SetAreaOfABP (CallFunction|CalculateTriangleArea TrianglePointA TrianglePointB PointPosition)))
      (:then_3
        ;;[1 1 1] Area of BCP
        (LocalVariables|SetAreaOfBCP (CallFunction|CalculateTriangleArea TrianglePointB TrianglePointC PointPosition)))
      (:then_4
        (Utilities|FlowControl|Sequence
          (:then_0
            ;[1 1 1] Calculate UVW
            ;;[1 1 1] Calculate U
            (LocalVariables|SetU (/ (LocalVariables|GetAreaOfCAP) (LocalVariables|GetAreaOfABC))))
          (:then_1
            ;;[1 1 1] Calculate V
            (LocalVariables|SetV (/ (LocalVariables|GetAreaOfABP) (LocalVariables|GetAreaOfABC))))
          (:then_2
            ;;[1 1 1] Calculate W
            (LocalVariables|SetW (/ (LocalVariables|GetAreaOfBCP) (LocalVariables|GetAreaOfABC))))
          (:then_3
            (return (LocalVariables|GetU) (LocalVariables|GetV) (LocalVariables|GetW) (<= (Math|Float|float+float (LocalVariables|GetU) (LocalVariables|GetV) (LocalVariables|GetW)) 1.001) true)))))
    (else
      (Development|PrintString 0 "Point in ConvertIntoVarycentric not Coplaner" false)
      (return "0, 0, 0" false false))))

(fn ConvertOutOfBarycentric
  (params (BarycentricCoords_X :type Float%20%28double-precision%29) (BarycentricCoords_Y :type Float%20%28double-precision%29) (BarycentricCoords_Z :type Float%20%28double-precision%29) (TargetTrianglePointA :type Vector) (TargetTrianglePointB :type Vector) (TargetTrianglePointC :type Vector))
  (return (Math|Vector|vector+vector (* TargetTrianglePointA BarycentricCoords_Z) (* TargetTrianglePointB BarycentricCoords_X) (* TargetTrianglePointC BarycentricCoords_Y))))
