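; Helper function library, likely for the editor utility widgets of the art review SunTool.
; CreateSliderLabel paints colored text at an offset from a slider's top-left corner, and LerpVector2D interpolates 2D vectors through a 3D vector lerp.
; PointVectorPointVectorIntersection finds where two 3D lines meet, using normalized cross products to test intersection and pick direction, and length ratios for distance.
; It returns the point with a success flag and prints debug strings.
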(fn CreateSliderLabel
  (params (OnPaintContext :type Paint%20Context%20Structure) (Slider :type Slider%20Object%20Reference) (PositionalShift :type Vector%202D%20Structure) (Label :type String) (Label%20Color :type Linear%20Color%20Structure :default "(R=1.000000,G=1.000000,B=1.000000,A=1.000000)") (FontSize :type Integer :default 5))
  (Painting|DrawText OnPaintContext (Utilities|Text|ToText(String) Label) (+ (UserInterface|Geometry|GetLocalTopLeft (Widget|GetCachedGeometry Slider)) PositionalShift) "/FortUILibrary/Fonts/NotoSans.NotoSans" (Math|Conversions|ToFloat(Integer) FontSize) "Regular" Label%20Color)
  (return OnPaintContext))

(fn LerpVector2D
  (params (A :type Vector%202D%20Structure) (B :type Vector%202D%20Structure) (Alpha :type Float%20%28double-precision%29))
  (bind (_returnvalue_x _returnvalue_y) (Math|Vector|BreakVector (Math|Vector|Lerp(Vector) (Math|Vector|MakeVector (Math|Vector2D|BreakVector2D A) (.Y (Math|Vector2D|BreakVector2D A))) (Math|Vector|MakeVector (Math|Vector2D|BreakVector2D B) (.Y (Math|Vector2D|BreakVector2D B))) Alpha)))
  (bind (_x _y) (Math|Vector2D|BreakVector2D A))
  (bind (_x_2 _y_2) (Math|Vector2D|BreakVector2D B))
  (return _returnvalue_x _returnvalue_y))

(fn PointVectorPointVectorIntersection
  (params (FirstVectorOrigin :type Vector) (FirstVectorDirection :type Vector) (SecondVectorOrigin :type Vector) (SecondVectorDirection :type Vector))
  (bind _returnvalue (- FirstVectorOrigin (* FirstVectorDirection (/ (Math|Vector|VectorLength (Math|Vector|CrossProduct SecondVectorDirection (- SecondVectorOrigin FirstVectorOrigin))) (Math|Vector|VectorLength (Math|Vector|CrossProduct SecondVectorDirection FirstVectorDirection))))))
  (bind _returnvalue_2 (- SecondVectorOrigin FirstVectorOrigin))
  (bind _returnvalue_3 (Math|Vector|CrossProduct SecondVectorDirection _returnvalue_2))
  (bind _returnvalue_4 (Math|Vector|VectorLength _returnvalue_3))
  (bind _returnvalue_5 (Math|Vector|CrossProduct SecondVectorDirection FirstVectorDirection))
  (bind _returnvalue_6 (Math|Vector|VectorLength _returnvalue_5))
  (bind _returnvalue_7 (/ _returnvalue_4 _returnvalue_6))
  (bind _returnvalue_8 (* FirstVectorDirection _returnvalue_7))
  (bind _returnvalue_9 (+ FirstVectorOrigin _returnvalue_8))
  (if (or (Math|Vector|Equal(Vector) (+ (Math|Vector|Normalize (Math|Vector|CrossProduct (- FirstVectorOrigin SecondVectorOrigin) FirstVectorDirection)) (Math|Vector|Normalize (Math|Vector|CrossProduct (- FirstVectorOrigin SecondVectorOrigin) SecondVectorDirection))) "0, 0, 0" 0.0001) (Math|Vector|Equal(Vector) (- (Math|Vector|Normalize (Math|Vector|CrossProduct (- FirstVectorOrigin SecondVectorOrigin) FirstVectorDirection)) (Math|Vector|Normalize (Math|Vector|CrossProduct (- FirstVectorOrigin SecondVectorOrigin) SecondVectorDirection))) "0, 0, 0" 0.0001))
    (if (Math|Vector|Equal(Vector) (- (Math|Vector|Normalize _returnvalue_3) (Math|Vector|Normalize _returnvalue_5)) "0, 0, 0" 0.0001)
      (Development|PrintString 0 (Utilities|String|ToString(Vector) _returnvalue_9) false)
      (return _returnvalue_9 true)
      (else
        (Development|PrintString 0 (Utilities|String|ToString(Vector) _returnvalue) false)
        (return _returnvalue true)))
    (else
      (Development|PrintString 0 "LINES DO NOT INTERSECT" false)
      (return "0, 0, 0" false))))
