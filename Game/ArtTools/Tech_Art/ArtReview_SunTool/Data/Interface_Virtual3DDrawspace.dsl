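; Blueprint interface for a 3D virtual drawspace in the ArtReview SunTool art tool.
; It declares calls to add, edit, remove and clear colored lines between vector endpoints, with an integer index addressing lines for edits and removal.
; It also declares updates for an overwritten state with sun Y and Z vectors, and for left, right, top and bottom spacers.
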(fn AddLine
  (params (LineStart :type Vector) (LineEnd :type Vector) (Color :type Linear%20Color%20Structure))
  (return 0))

(fn ClearLines)

(fn EditLine
  (params (Index :type Integer) (LineStart :type Vector) (LineEnd :type Vector) (Color :type Linear%20Color%20Structure)))

(fn RemoveLine
  (params (Index :type Integer)))

(fn UpdateOverWritten
  (params (IsOverWritten? :type Boolean) (SunY :type Vector) (SunZ :type Vector)))

(fn UpdateSpacers
  (params (LeftSpacer :type Float%20%28double-precision%29) (RightSpacer :type Float%20%28double-precision%29) (TopSpacer :type Float%20%28double-precision%29) (BottomSpacer :type Float%20%28double-precision%29)))
